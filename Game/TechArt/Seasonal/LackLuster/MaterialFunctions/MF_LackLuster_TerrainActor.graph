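# Material function that applies the LackLuster seasonal landscape treatment to terrain actor materials, behind the UseLacklusterTerrainActor static switch that otherwise passes attributes through.
# Input attributes and height go through the Landscape_RVT function, then Landscape_PostRVT, sharing one mask.
# UseVertexColorBlending selects vertex color or a constant 1 as that mask and also toggles landscape layer blending.
# All four inputs to the channel selector are wired to vertex red, so VertexColorBlendingChannel always receives red.
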MaterialFunction /Game/TechArt/Seasonal/LackLuster/MaterialFunctions/MF_LackLuster_TerrainActor
graph {nodes [
  n1 = FunctionInput() {inputName=MA_In, inputType=FunctionInput_MaterialAttributes}
  n2 = VertexColor()
  n3 = ChannelMaskParameter(None=MaterialFunctionCall(R=@n2.R, G=@n2.R, B=@n2.R, A=@n2.R) {materialFunction=&AppendMany}.RGBA) {group=LacklusterTerrainActor, maskChannel=Blue, parameterName=VertexColorBlendingChannel, defaultValue=<rgba 0 0 1 0>}
  n4 = StaticSwitchParameter(True=@n3.R, False=Constant() {r=1}) {group=LacklusterTerrainActor, parameterName=UseVertexColorBlending}
  n5 = FunctionInput() {bUsePreviewValueAsDefault=true, inputName="Height(alpha)", inputType=FunctionInput_Scalar, sortPriority=1, previewValue=<0.5 0 0 1>}
  n6 = MaterialFunctionCall!(
   MA=@n1
   Mask=@n4
   HeightIn=@n5
   SlopeMask
   BlendLandscapeLayers=StaticSwitchParameter(True=StaticBool() {value=true}, False=StaticBool()) {group=LacklusterTerrainActor, parameterName=UseVertexColorBlending}
  ) {materialFunction=&/MF_LackLuster_Landscape_RVT}
  n7 = StaticBool()
  FunctionOutput(None=StaticSwitchParameter(
    True=MaterialFunctionCall!(In=@n6.Result, Mask=@n4, UseCheapSlopeMask=@n7, MaskReflectionCurveOnZ, HueShiftOverride) {materialFunction=&/MF_LackLuster_Landscape_PostRVT}.Result
    False=@n1
   ) {group=LacklusterTerrainActor, parameterName=UseLacklusterTerrainActor}) {outputName=MA_Out}
 ]}
