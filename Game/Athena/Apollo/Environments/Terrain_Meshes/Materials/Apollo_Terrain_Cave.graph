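# Cave terrain material for Apollo static and spline meshes, with cliff walls, rock ceilings, a gravel layer and Nanite support.
# Downward-facing world normals height-blend cliff walls into EdgeRocks ceilings, while mobile swaps in a cheaper route reusing cliff XYZ projection.
# Vertex colors drive rock tint (G), a material blend (B) and height-blended floor transition (A).
# The result passes through LackLuster seasonal and Creative move-tool functions.
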Material /Game/Athena/Apollo/Environments/Terrain_Meshes/Materials/Apollo_Terrain_Cave
graph {
 properties {
  bTangentSpaceNormal=false
  bUseMaterialAttributes=true
  bUsedWithInstancedStaticMeshes=true
  bUsedWithNanite=true
  bUsedWithSplineMeshes=true
  ditheredLODTransition=true
  physMaterial=/Game/FortressPhysicalMaterials/Dirt
  pixelDepthOffsetMode=PDOM_Legacy
  usedShadingModels=Default Lit
 }
 nodes [
  BlendMaterialAttributes(A=NamedRerouteUsage(), B=NamedRerouteUsage(), Alpha=VertexColor().B)
  Constant() {r=669.0363}
  Constant() {r=582}
  Constant()
  Constant() {r=1.7534}
  LinearInterpolate(
   A=VectorParameter() {parameterName=ColorTint_RockCave, defaultValue=<rgba 0.7083 0.4385 0.1586 1>}.RGB
   B=VectorParameter() {parameterName=ColorTint_RockCliff, defaultValue=<rgba 0.8698 0.8589 0.6537 1>}.RGB
   Alpha=VertexColor().G
  )
  n1 = NamedRerouteUsage()
  n2 = NamedRerouteUsage()
  n3 = MaterialFunctionCall(Transition Phase=VertexColor().A, Height Texture=NamedRerouteUsage(), Contrast=ScalarParameter() {parameterName=ContrastFloorTransition}) {materialFunction=&HeightLerp}
  MaterialFunctionCall(MA In=MaterialFunctionCall!(MA_In=BlendMaterialAttributes(A=@n1, B=@n2, Alpha=@n3.Alpha), "Height(alpha)") {materialFunction=/Game/TechArt/Seasonal/LackLuster/MaterialFunctions/MF_LackLuster_TerrainActor}.MA_Out) {materialFunction=&MF_CreativeMoveTool}
  NamedRerouteDeclaration(None=MaterialFunctionCall!(OverrideEdgeColor, EdgeMask, BypassBiomeColors) {materialFunction=/Game/Athena/Apollo/Environments/Landscape/Materials/Biome/CommonMaterials/Mobile/MF_Mobile_Common_Gravel}.ForStaticMesh) {name=Gravel-SM-Material, nodeColor=<rgba 1 0 1 1>}
  n4 = MaterialFunctionCall!(ColorTint=NamedRerouteUsage(), TilingSize) {materialFunction=/Game/Athena/Apollo/Environments/Terrain_Meshes/MaterialFunctions/MF_Common_Clff_StaticMesh}
  NamedRerouteDeclaration(None=@n4.Result) {name=Cliff-SM-Material, nodeColor=<rgba 1 0 1 1>}
  NamedRerouteDeclaration(None=@n4.XYZ-Material) {name=Cliff-SM-XYZ-Material, nodeColor=<rgba 1 0 1 1>}
  NamedRerouteDeclaration(None=@n4.Height) {name=Cliff-SM-Material-Height, nodeColor=<rgba 1 0 1 1>}
  n5 = MaterialFunctionCall!(
   ColorTint=NamedRerouteUsage()
   TilingSize=ScalarParameter() {defaultValue=582, parameterName=SizeRockCeiling}
   Use_Shift_Worldposition=StaticBool()
  ) {materialFunction=/Game/Athena/Apollo/Environments/Terrain_Meshes/MaterialFunctions/MF_EdgeRocks}
  n6 = ComponentMask(None=VertexNormalWS()) {b=true}
  n7 = Power(Base=Saturate(None=Multiply(A=@n6) {constB=-1}), Exp=ScalarParameter() {defaultValue=1.7534, parameterName=CeilingTransition})
  n8 = ScalarParameter() {parameterName=ContrastCeilingTransition}
  n9 = ShadingPathSwitch(
   Default=BlendMaterialAttributes(
    A=NamedRerouteUsage()
    B=@n5."Z Ceiling "
    Alpha=MaterialFunctionCall(Transition Phase=@n7, Height Texture=@n5.Z Ceiling Height, Contrast=@n8) {materialFunction=&HeightLerp}.Alpha
   )
   Mobile=NamedRerouteUsage()
  ) {desc="Use the XYZ projection of the default cliff textures for the cave\r\nceiling instead of a unique texture set that isn't (?) used anywhere\r\nelse on mobile"}
  NamedRerouteDeclaration(None=@n9) {name="CliffWall&&CeilingMaterial", nodeColor=<rgba 1 0 1 1>}
  NamedRerouteDeclaration(None=@n4.XYZ-Mobile-Cave) {name=Cliff-SM-XYZ-MobileMaterial, nodeColor=<rgba 1 0 1 1>}
  NamedRerouteDeclaration(None=MaterialFunctionCall!(DarknessAmount=Constant() {r=0.2}) {materialFunction=/Game/Athena/Apollo/Environments/Landscape/Materials/Biome/CommonMaterials/MF_HeightBasedDarkenCliff}.Result) {name=CliffDarken, nodeColor=<rgba 1 0 1 1>}
  ScalarParameter() {defaultValue=1000, desc="This should match the same value as what is used in MF_VextexColorLayerBlend!", parameterName=SizeRockCliffs}
 ]
}
